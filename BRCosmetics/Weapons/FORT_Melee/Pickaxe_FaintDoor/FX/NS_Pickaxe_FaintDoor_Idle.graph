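# Idle Niagara effect for the FaintDoor pickaxe, adding a socket-attached glow and smoke.
# NE_Light bursts socket-following sprites with lights on its first loop.
# User.bisSwinging switches smoke from NE_Smoke_Local, which follows the socket and drifts up at a steady rate, to NE_Smoke_World, which spawns per distance moved and trails free with curl noise, upward force and drag.
# User.Variant, converted to an integer, selects one of six color palettes per emitter.
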NiagaraSystem /BRCosmetics/Weapons/FORT_Melee/Pickaxe_FaintDoor/FX/NS_Pickaxe_FaintDoor_Idle
dependencies {
 usedDataInterfaces [
  &NiagaraDataInterfaceSpriteRendererInfo
  &NiagaraDataInterfaceCurve
  &NiagaraDataInterfaceSocketReader
 ]
 usedDynamicInputs []
 usedModules [
  &self:SystemSpawnScript.SystemScriptSource.SystemScriptGraph.NiagaraNodeAssignment_0.SetVariables_#1
  &SystemState
 ]
}
emitters {
 NE_Light {
  emitterUpdateScript [
   &EmitterState(Life Cycle Mode=::NewEnumerator1, Loop Delay=1, Loop Duration=0.1, UseLoopDelay=false)
   &SpawnBurst_Instantaneous(Spawn Probability=HlslExpression() {hlslExpression="Emitter.LoopCount == 0"}, Use Spawn Probability=true, UseLoopCountLimit=false)
  ]
  particleSpawnScript [
   /SetVariables(Particles.VisibilityTag=&SpawnGroup()) {isSetParameters=true}
   &InitializeParticle(
    Color=&MergeHelper_ColorSwitch(
     0=<rgba 5 0.4181 0.0016 0.25>
     1=<rgba 4.7825 0 5 0.25>
     2=<rgba 5 1.75 0 0.25>
     3=<rgba 5 1 4.7566 0.25>
     4=<rgba 0 2.0833 5 0.25>
     5=<rgba 0.0435 2.1087 5 0.25>
     Index=&MakeIntFromFloat(Float=$User.Variant)
     Num=::NewEnumerator4
    )
    Position Offset Coordinate Space=::Local
    Sprite Size Mode=::NewEnumerator3
    Uniform Sprite Size=50
    UsePositionOffset=true
    Write Material Random=true
   )
   &CameraOffset(Camera Offset Amount=&MergeHelper_FloatSwitch(0=0, 1=10, 2=?, 3=?, Index=&SpawnGroup(), Num=::NewEnumerator0))
  ]
  particleUpdateScript [
   &ParticleState(Kill Particles When Lifetime Has Elapsed=false, Let Infinitely Lived Particles Die When Emitter Deactivates=true)
   &SocketLocation(SamplePosOffset=true, Socket Reader=$System.Socket Reader, Socket Sampling Mode=::NewEnumerator3, VelocitySpeedLimit=false)
   &FN_Light_Attributes(
    Color=&MergeHelper_ColorSwitch(
     0=<rgba 50 7.8261 0 1>
     1=<rgba 26.9559 0 50 1>
     2=<rgba 50 33.9129 0 1>
     3=<rgba 50 10 47.5658 1>
     4=<rgba 0 0 25 1>
     5=<rgba 0.4348 8.6957 50 1>
     Index=&MakeIntFromFloat(Float=$User.Variant)
     Num=::NewEnumerator4
    )
    Create Color Attribute=true
    Create Exponent Attribute=true
    Create Position Attribute=true
    Create Radius Attribute=true
    Exponent=2
    Position=&AddVectorToPosition(Position=$Particles.Position, Vector=<0 0 0>)
    Radius=40
    Scale Color=0.5
   )
  ]
  rendererClasses [&NiagaraSpriteRendererProperties, &NiagaraLightRendererProperties]
  renderers [
   {rendererClass=&NiagaraSpriteRendererProperties}
   {rendererClass=&NiagaraLightRendererProperties, rendererIndex=1}
  ]
 }
 NE_Smoke_Local {
  emitterUpdateScript [
   &EmitterState(Life Cycle Mode=::NewEnumerator1, Loop Delay=1, Loop Duration=1, UseLoopDelay=false)
   &SpawnRate(
    Spawn Probability=&MakeCustomFloatFromBool(Bool=$User.bisSwinging, True Float=0, False Float=1)
    SpawnRate=5
    Use Spawn Probability=true
   )
  ]
  particleSpawnScript [
   &InitializeParticle(
    Color=&MergeHelper_ColorSwitch(
     0=<rgba 5 0.4181 0.0016 0.25>
     1=<rgba 4.7825 0 5 0.25>
     2=<rgba 5 2.0833 0 0.25>
     3=<rgba 5 1 4.7566 0.25>
     4=<rgba 1 2.3913 5 0.25>
     5=<rgba 0.0435 2.1087 5 0.25>
     Index=&MakeIntFromFloat(Float=$User.Variant)
     Num=::NewEnumerator4
    )
    Lifetime=&RandomRangeFloat(Minimum=1, Maximum=2)
    Position Offset=<0 0 0.5>
    Sprite Rotation Mode=::NewEnumerator2
    Sprite Size Mode=::NewEnumerator1
    Uniform Sprite Size Max=30
    Uniform Sprite Size Min=15
    UsePositionOffset=false
    Write Material Random=true
   )
   &CameraOffset(Camera Offset Amount=10)
  ]
  particleUpdateScript [
   &ParticleState()
   &SocketLocation(SamplePosOffset=true, Socket Reader=$System.Socket Reader, Socket Sampling Mode=::NewEnumerator3, VelocitySpeedLimit=false)
   &SubUVAnimation(
    End Frame Range Override=64
    "Sprite Renderer "=DataInterface() {propertyValues {SpriteRenderer=&self:NE_Smoke_Local_0.NiagaraSpriteRendererProperties_1}}
    UseEndFrame=true
    UseStartFrame=true
   )
   &ScaleColor(
    Scale Alpha=&FloatFromCurve(FloatCurve=DataInterface() {propertyValues {Curve {keys [<at 0 0>,<at 0.5 1>,<at 1 0>]}}}, Scale Curve=0.15)
    ScaleA=true
    ScaleRGB=true
   )
   &SolveForcesAndVelocity(Clamp Velocity=false, Limit Acceleration=false)
   &OffsetPosition(
    Position Offset=&Lerp_Vector(
     A=<0 0 0>
     B=<0 0 25>
     Alpha=&FloatFromCurve(FloatCurve=DataInterface() {propertyValues {Curve {defaultValue=1, keys [{tangentMode=RCTM_User, time=0, value=0}, {tangentMode=RCTM_User, time=1, value=1}]}}})
    )
    Position Offset Coordinate Space:ENiagaraCoordinateSpace=::Local
   )
  ]
  rendererClasses [&NiagaraSpriteRendererProperties]
  renderers [{rendererClass=&NiagaraSpriteRendererProperties}]
 }
 NE_Smoke_World {
  emitterUpdateScript [
   &EmitterState(Life Cycle Mode=::NewEnumerator1, Loop Delay=1, Loop Duration=1, UseLoopDelay=false)
   &SpawnPerUnit(
    Module.Use Max Movement Threshold=true
    Module.Use Movement Tolerance=false
    Spawn Probability:NiagaraFloat=&MakeCustomFloatFromBool(Bool=$User.bisSwinging, True Float=1, False Float=0)
    Spawn Spacing=30
    Use Spawn Probability=true
   )
  ]
  particleSpawnScript [
   &InitializeParticle(
    Color=&MergeHelper_ColorSwitch(
     0=<rgba 5 0.4181 0.0016 0.25>
     1=<rgba 4.7825 0 5 0.25>
     2=<rgba 5 2.0833 0 0.25>
     3=<rgba 2.5 0.5 2.3783 0.25>
     4=<rgba 1 2.3913 5 0.25>
     5=<rgba 0.0435 0.8696 5 0.25>
     Index=&MakeIntFromFloat(Float=$User.Variant, Conversion Mode=::NewEnumerator1)
     Num=::NewEnumerator4
    )
    Lifetime=&RandomRangeFloat(Minimum=0.3, Maximum=0.5)
    Position Offset=<0 0 0.5>
    Sprite Rotation Mode=::NewEnumerator2
    Sprite Size Mode=::NewEnumerator1
    Uniform Sprite Size Max=80
    Uniform Sprite Size Min=20
    UsePositionOffset=false
    Write Material Random=true
   )
   &SocketLocation(SamplePosOffset=false, Socket Reader=$System.Socket Reader, Socket Sampling Mode=::NewEnumerator3, VelocitySpeedLimit=false)
   &CameraOffset(Camera Offset Amount=3)
  ]
  particleUpdateScript [
   &ParticleState()
   &SubUVAnimation(
    End Frame Range Override=64
    "Sprite Renderer "=DataInterface() {propertyValues {SpriteRenderer=&self:NE_Smoke_World_0.NiagaraSpriteRendererProperties_1}}
    SubUV Animation Mode:ENiagara_SubUVLookupModeV2=::NewEnumerator3
    UseEndFrame=true
    UsePlayRateScale=false
    UseStartFrame=true
   )
   &CurlNoiseForce(Bias Noise=false, Noise Strength=100, Pan Noise=false)
   &GravityForce(Gravity=&RandomRangeVector(Minimum=<0 0 150>, Maximum=<0 0 300>))
   &ScaleColor(
    Scale Alpha=&FloatFromCurve(FloatCurve=DataInterface() {propertyValues {Curve {keys [<at 0 0>,<at 0.5 1>,<at 1 0>]}}}, Scale Curve=0.15)
    Scale RGB=&MakeVectorFromLinearColorRGB(Color=&MergeHelper_ColorSwitch(
      0=<rgba 1 1 1 1>
      1=<rgba 1 1 1 1>
      2=<rgba 1 1 1 1>
      3=&RandomRangeLinearColor(Minimum=<rgba 2.5 0.5 2.3783 0.25>, Maximum=<rgba 0.5 1.5087 2.5 0.25>)
      4=<rgba 1 1 1 1>
      5=<rgba 1 1 1 1>
      Index=&MakeIntFromFloat(Float=$User.Variant)
      Num=::NewEnumerator4
     ))
    ScaleA=true
    ScaleRGB=true
   )
   &Drag(Drag=15, Use Linear Drag=true, Use Rotational Drag=true)
   &SolveForcesAndVelocity(Clamp Velocity=false, Limit Acceleration=false)
  ]
  rendererClasses [&NiagaraSpriteRendererProperties]
  renderers [{rendererClass=&NiagaraSpriteRendererProperties}]
 }
}
summary {
 systemName=NS_Pickaxe_FaintDoor_Idle
 emitters [
  {emitterName=NE_Light, rendererClasses [&NiagaraSpriteRendererProperties, &NiagaraLightRendererProperties]}
  {emitterName=NE_Smoke_Local, rendererClasses [&NiagaraSpriteRendererProperties]}
  {emitterName=NE_Smoke_World, rendererClasses [&NiagaraSpriteRendererProperties]}
 ]
 userVariables [
  User.bIsFrontEnd {type=NiagaraBool}
  User.bIsFrontEndPreview {type=NiagaraBool}
  User.bisSwinging {type=NiagaraBool}
  User.PhysicalSurface {type=NiagaraInt32}
  User.RibbonIDOffset {type=NiagaraInt32}
  User.SwingSectionIndex {type=NiagaraInt32}
  User.Variant {type=NiagaraFloat, defaultValue=0, description=""}
 ]
}
