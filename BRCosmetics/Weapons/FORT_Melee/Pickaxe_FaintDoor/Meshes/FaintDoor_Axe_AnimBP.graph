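# Animation Blueprint for the FaintDoor pickaxe, simulating its secondary motion, including a whip bone chain.
# Each update copies the owner's root, pelvis, thigh and calf transforms onto matching bones before RigidBody simulation using borrowed pickaxe physics assets.
# Front-end display adds a second simulation, chosen by whether it is on a pawn.
# Trail controllers replace rigid bodies when RigidsActive is false; PhysicsOff curves fade physics, and swimming transitions reset dynamics.
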AnimBlueprint /BRCosmetics/Weapons/FORT_Melee/Pickaxe_FaintDoor/Meshes/FaintDoor_Axe_AnimBP
target_skeleton=&/FaintDoor_Axe_Skeleton
graphs [
 AnimGraph {nodes [
   n1 = RigidBody(ComponentPose=Local To Component(LocalPose=Use cached pose 'PreFEPhysics'()), Alpha=Get RigidsFE_Alpha() -> (RigidsFE_Alpha:double)) {
    bOverrideWorldGravity=true
    overridePhysicsAsset=/BRCosmetics/Weapons/FORT_Melee/Pickaxe_Embers_Male/Meshes/Embers_Male_Pickaxe_FE_Physics
    simulationSpace=BaseBoneSpace
    alphaBoolBlend {blendOption=Linear}
    alphaScaleBiasClamp {interpSpeedDecreasing=20, interpSpeedIncreasing=20}
    baseBoneRef {boneName=root}
    componentAppliedLinearAccClamp=<3000 3000 3000>
    componentLinearAccScale {z=0.25}
    componentLinearVelScale {z=0.25}
    overrideWorldGravity {z=-980}
    simSpaceSettings {velocityScaleZ=0}
   }
   n2 = RigidBody(ComponentPose=Local To Component(LocalPose=Use cached pose 'PreFEPhysics'()), Alpha=Get RigidsFE_Alpha()) {
    bClampLinearTranslationLimitToRefPose=true
    bOverrideWorldGravity=true
    overridePhysicsAsset=/BRCosmetics/Weapons/FORT_Melee/Pickaxe_Fruitcake_Female/Meshes/FruitCakeFemale_Pickaxe_FE_OnPawn_Physics
    simulationSpace=WorldSpace
    alphaBoolBlend {blendOption=Linear}
    alphaScaleBiasClamp {interpSpeedDecreasing=20, interpSpeedIncreasing=20}
    baseBoneRef {boneName=b_AR_Root}
    overrideWorldGravity {z=-1470}
    simSpaceSettings {maxAngularAcceleration=40, maxAngularVelocity=8, maxLinearAcceleration=3500, maxLinearVelocity=700, worldAlpha=1, externalLinearDragV=<3 3>}
   }
   n3 = Blend Poses by bool(
    BlendPose_0=Component To Local(ComponentPose=@n2)
    BlendPose_1=Component To Local(ComponentPose=@n1)
    bActiveValue=Get bIsOnPawn() -> (bIsOnPawn:bool)
   ) {blendType=Linear}
   n4 = Use cached pose 'PreFEPhysics'()
   n5 = Get bIsInFrontEnd()
   n6 = Get bIsInFrontEnd()
   n7 = "Transform (Modify) Bone - Bone: whip_skin_01"!(
    ComponentPose=Local To Component(LocalPose=Use cached pose 'InputCache'())
    bAlphaBoolEnabled:bool=@n6
    Translation="0,0,0"
    Rotation="-40,0.000,0.000"
    Scale="1,1,1"
   ) -> (Pose) {
    alphaInputType=Bool
    rotationMode=BMM_Additive
    rotationSpace=BCS_BoneSpace
    alphaBoolBlend {blendOption=Linear}
    boneToModify {boneName=whip_skin_01}
    rotation {pitch=-40}
   }
   n8 = Trail controller - Bone: whip_skin_10!(ComponentPose=@n7, bAlphaBoolEnabled:bool=@n6) -> (Pose) {
    alphaInputType=Bool
    chainLength=10
    relaxationSpeedScale=2.5
    rotationLimits [
     {limitMax=<180 180 180>}
     {limitMax=<180 180 180>}
     {limitMax=<180 180 180>}
     {limitMax=<180 180 180>}
     {limitMax=<180 180 180>}
     {limitMax=<180 180 180>}
     {limitMax=<180 180 180>}
     {limitMax=<180 180 180>}
     {limitMax=<180 180 180>}
     {limitMax=<180 180 180>}
    ]
    rotationOffsets [<0 0 0>, <0 0 0>, <0 0 0>, <0 0 0>, <0 0 0>, <0 0 0>, <0 0 0>, <0 0 0>, <0 0 0>, <0 0 0>]
    trailBone {boneName=whip_skin_10}
    bLimitStretch=~
   }
   n9 = Trail controller - Bone: whip_skin_10!(ComponentPose=@n8, Alpha:float=Get NegPhysicsAlpha() -> (NegPhysicsAlpha:double)) -> (Pose) {
    chainLength=10
    maxDeltaTime=0.0333
    relaxationSpeedScale=2.5
    rotationLimits [
     {limitMax=<180 180 180>}
     {limitMax=<180 180 180>}
     {limitMax=<180 180 180>}
     {limitMax=<180 180 180>}
     {limitMax=<180 180 180>}
     {limitMax=<180 180 180>}
     {limitMax=<180 180 180>}
     {limitMax=<180 180 180>}
     {limitMax=<180 180 180>}
     {limitMax=<180 180 180>}
    ]
    rotationOffsets [<0 0 0>, <0 0 0>, <0 0 0>, <0 0 0>, <0 0 0>, <0 0 0>, <0 0 0>, <0 0 0>, <0 0 0>, <0 0 0>]
    trailBone {boneName=whip_skin_10}
   }
   n10 = "Transform (Modify) Bone - Bone: whip_skin_01"(ComponentPose=@n9, bAlphaBoolEnabled=Get FrEnd_onPawn() -> (FrEnd_onPawn:bool), Translation="0,0,0", Rotation="40,0,30", Scale="1,1,1") {alphaInputType=Bool, rotationMode=BMM_Additive, boneToModify {boneName=whip_skin_01}, rotation {pitch=40, roll=30}}
   Output Pose(Result=Blend Poses by bool(
     BlendPose_0=Blend Poses by bool(BlendPose_0=@n3, BlendPose_1=@n4, bActiveValue=@n5) {blendType=Linear}
     BlendPose_1=Component To Local(ComponentPose=@n10)
     bActiveValue=Get RigidsActive() -> (RigidsActive:bool)
    ) {blendType=Linear}) {layerGroup=None}
   AnimGraphNode_SubInput_0 = Input Pose()
   Save cached pose 'InputCache'(Pose=@AnimGraphNode_SubInput_0) {cacheName=InputCache}
   n11 = "Transform (Modify) Bone - Bone: dyn_root_simspace"(
    ComponentPose=Local To Component(LocalPose=Use cached pose 'InputCache'())
    Alpha=1
    Translation="0,0,0"
    Rotation=Get root_space_rotation() -> (root_space_rotation:Rotator)
    Scale="1,1,1"
   ) {
    rotationMode=BMM_Replace
    rotationSpace=BCS_WorldSpace
    translationSpace=BCS_BoneSpace
    alphaBoolBlend {blendOption=Linear}
    alphaScaleBiasClamp {bInterpResult=true, interpSpeedDecreasing=2, interpSpeedIncreasing=2}
    boneToModify {boneName=dyn_root_simspace}
   }
   n12 = "Transform (Modify) Bone - Bone: body_root"(
    ComponentPose=@n11
    Alpha=1
    Translation=Get body_root_location() -> (body_root_location:Vector)
    Rotation=Get body_root_rotation() -> (body_root_rotation:Rotator)
    Scale="1,1,1"
   ) {
    rotationMode=BMM_Replace
    rotationSpace=BCS_WorldSpace
    scaleSpace=BCS_BoneSpace
    translationMode=BMM_Replace
    translationSpace=BCS_WorldSpace
    alphaBoolBlend {blendOption=Linear}
    alphaScaleBiasClamp {bInterpResult=true, interpSpeedDecreasing=1, interpSpeedIncreasing=1, inRange {max=0, min=-40}, outRange {max=0, min=1}}
    boneToModify {boneName=body_root}
   }
   n13 = "Transform (Modify) Bone - Bone: pelvis"(
    ComponentPose=@n12
    Alpha=1
    Translation=Get pelvis_location() -> (pelvis_location:Vector)
    Rotation=Get pelvis_rotation() -> (pelvis_rotation:Rotator)
    Scale="1,1,1"
   ) {
    rotationMode=BMM_Replace
    rotationSpace=BCS_WorldSpace
    scaleSpace=BCS_BoneSpace
    translationMode=BMM_Replace
    translationSpace=BCS_WorldSpace
    alphaBoolBlend {blendOption=Linear}
    alphaScaleBiasClamp {inRange {max=0, min=-40}, outRange {max=0, min=1}}
    boneToModify {boneName=pelvis}
   }
   n14 = "Transform (Modify) Bone - Bone: thigh_l"(
    ComponentPose=@n13
    Alpha=1
    Translation=Get thigh_left_location() -> (thigh_left_location:Vector)
    Rotation=Get thigh_left_rotation() -> (thigh_left_rotation:Rotator)
    Scale="1,1,1"
   ) {
    rotationMode=BMM_Replace
    rotationSpace=BCS_WorldSpace
    scaleSpace=BCS_BoneSpace
    translationMode=BMM_Replace
    translationSpace=BCS_WorldSpace
    alphaBoolBlend {blendOption=Linear}
    alphaScaleBiasClamp {inRange {max=0, min=-40}, outRange {max=0, min=1}}
    boneToModify {boneName=thigh_l}
   }
   n15 = "Transform (Modify) Bone - Bone: thigh_r"(
    ComponentPose=@n14
    Alpha=1
    Translation=Get thigh_right_location() -> (thigh_right_location:Vector)
    Rotation=Get thigh_right_rotation() -> (thigh_right_rotation:Rotator)
    Scale="1,1,1"
   ) {
    rotationMode=BMM_Replace
    rotationSpace=BCS_WorldSpace
    scaleSpace=BCS_BoneSpace
    translationMode=BMM_Replace
    translationSpace=BCS_WorldSpace
    alphaBoolBlend {blendOption=Linear}
    alphaScaleBiasClamp {inRange {max=0, min=-40}, outRange {max=0, min=1}}
    boneToModify {boneName=thigh_r}
   }
   n16 = "Transform (Modify) Bone - Bone: calf_l"(
    ComponentPose=@n15
    Alpha=1
    Translation=Get calf_left_location() -> (calf_left_location:Vector)
    Rotation=Get calf_left_rotation() -> (calf_left_rotation:Rotator)
    Scale="1,1,1"
   ) {
    rotationMode=BMM_Replace
    rotationSpace=BCS_WorldSpace
    scaleSpace=BCS_BoneSpace
    translationMode=BMM_Replace
    translationSpace=BCS_WorldSpace
    alphaBoolBlend {blendOption=Linear}
    alphaScaleBiasClamp {inRange {max=0, min=-40}, outRange {max=0, min=1}}
    boneToModify {boneName=calf_l}
   }
   n17 = "Transform (Modify) Bone - Bone: calf_r"(
    ComponentPose=@n16
    Alpha=1
    Translation=Get calf_right_location() -> (calf_right_location:Vector)
    Rotation=Get calf_right_rotation() -> (calf_right_rotation:Rotator)
    Scale="1,1,1"
   ) {
    rotationMode=BMM_Replace
    rotationSpace=BCS_WorldSpace
    scaleSpace=BCS_BoneSpace
    translationMode=BMM_Replace
    translationSpace=BCS_WorldSpace
    alphaBoolBlend {blendOption=Linear}
    alphaScaleBiasClamp {inRange {max=0, min=-40}, outRange {max=0, min=1}}
    boneToModify {boneName=calf_r}
   }
   n18 = RigidBody(ComponentPose=@n17, Alpha=Get NegPhysicsAlpha()) {
    bClampLinearTranslationLimitToRefPose=true
    bOverrideWorldGravity=true
    overridePhysicsAsset=/BRCosmetics/Weapons/FORT_Melee/Pickaxe_Embers_Male/Meshes/Embers_Male_Pickaxe_Physics
    simulationSpace=BaseBoneSpace
    alphaBoolBlend {blendOption=Linear}
    alphaScaleBiasClamp {bInterpResult=true, interpSpeedDecreasing=0, interpSpeedIncreasing=0.8}
    baseBoneRef {boneName=Melee_TwoHanded_Handle}
    componentAppliedLinearAccClamp=<2000 2000 1000>
    overrideWorldGravity {z=-980}
    simSpaceSettings {
     dampingAlpha=0
     maxAngularAcceleration=400
     maxAngularVelocity=8
     maxLinearAcceleration=4000
     maxLinearVelocity=800
     worldAlpha=1
     externalLinearDragV=<4 4>
    }
   }
   Save cached pose 'PreFEPhysics'(Pose=Component To Local(ComponentPose=@n18)) {cacheName=PreFEPhysics}
  ]}
 EventGraph {nodes [
   Event BlueprintUpdateAnimation()
   n1 = GetPartSkeletalMeshComponent()
   > Is Valid(InputObject=@n1)
   n2 = GetAttachParent!(self:Scene Component Object Reference=@n1) -> (ReturnValue:Scene Component Object Reference)
   > Is Valid(InputObject=@n2)
   > Set Body_Mesh(Body_Mesh:Scene Component Object Reference=@n2) -> (Output_Get:Scene Component Object Reference)
   > n3 = Sequence!(execute) -> (then_0, then_1, then_2:Exec, then_3:Exec, then_4:Exec, then_5:Exec, then_6:Exec, then_7:Exec, then_8:Exec, then_9:Exec, then_10:Exec)
   n4 = Get Body_Mesh() -> (Body_Mesh:Scene Component Object Reference)
   n5 = BreakTransform(InTransform=GetSocketTransform(self=@n4, InSocketName=root, TransformSpace=RTS_World))
   Set root_location(execute=@n3.then_7, root_location:Vector=@n5) -> (Output_Get:Vector)
   > Set root_space_rotation(root_space_rotation:Rotator=@n5.Rotation) -> (Output_Get:Rotator)
   n6 = Get CurrentLOD()
   n7 = Branch(execute=@n3.then_8, Condition=OR Boolean(A=Get bRigidBodyEnabled(), B="integer > integer"(A=@n6, B=1) {is=title K2Node_CallFunction}))
   > Set RigidsActive(RigidsActive:bool=true) -> (Output_Get:bool)
   Set RigidsActive(execute=@n7.else, RigidsActive=false)
   n8 = Branch(execute=@n3.then_9, Condition=Get bIsInFrontEnd())
   > Set RigidsFE_Alpha(RigidsFE_Alpha:double=MapRangeClamped(Value=GetCurveValue(CurveName=PhysicsOff), InRangeA=0, InRangeB=1, OutRangeA=1, OutRangeB=0)) -> (Output_Get:double)
   > Set NegPhysicsAlpha(NegPhysicsAlpha:double=0) -> (Output_Get:double)
   Set RigidsFE_Alpha(execute=@n8.else, RigidsFE_Alpha=0)
   Set NegPhysicsAlpha(
    execute=@n3.then_0
    NegPhysicsAlpha=MapRangeClamped(Value=GetCurveValue(CurveName=PhysicsOff), InRangeA=0, InRangeB=1, OutRangeA=1, OutRangeB=0)
   )
   > Set FrEnd_onPawn(FrEnd_onPawn:bool=AND Boolean(A=Get bIsOnPawn(), B=Get bIsInFrontEnd())) -> (Output_Get:bool)
   n9 = Get bIsPlayingMeleeAnim()
   n10 = Branch(execute=@n3.then_10, Condition=AND Boolean(A=Get bIsAccelerating2D(), B=NOT Boolean(A=@n9)))
   > Set rotateRun(rotateRun:bool=true) -> (Output_Get:bool)
   Set rotateRun(execute=@n10.else, rotateRun=false)
   n11 = BreakTransform(InTransform=GetSocketTransform(self=@n4, InSocketName=pelvis, TransformSpace=RTS_World))
   Set pelvis_location(execute=@n3.then_1, pelvis_location:Vector=@n11) -> (Output_Get:Vector)
   > Set pelvis_rotation(pelvis_rotation:Rotator=@n11.Rotation) -> (Output_Get:Rotator)
   n12 = BreakTransform(InTransform=GetSocketTransform(self=@n4, InSocketName=thigh_l, TransformSpace=RTS_World))
   Set thigh_left_location(execute=@n3.then_2, thigh_left_location:Vector=@n12) -> (Output_Get:Vector)
   > Set thigh_left_rotation(thigh_left_rotation:Rotator=@n12.Rotation) -> (Output_Get:Rotator)
   n13 = BreakTransform(InTransform=GetSocketTransform(self=@n4, InSocketName=thigh_r, TransformSpace=RTS_World))
   Set thigh_right_location(execute=@n3.then_3, thigh_right_location:Vector=@n13) -> (Output_Get:Vector)
   > Set thigh_right_rotation(thigh_right_rotation:Rotator=@n13.Rotation) -> (Output_Get:Rotator)
   n14 = BreakTransform(InTransform=GetSocketTransform(self=@n4, InSocketName=calf_r, TransformSpace=RTS_World))
   Set calf_right_location(execute=@n3.then_5, calf_right_location:Vector=@n14) -> (Output_Get:Vector)
   > Set calf_right_rotation(calf_right_rotation:Rotator=@n14.Rotation) -> (Output_Get:Rotator)
   n15 = BreakTransform(InTransform=GetSocketTransform(self=@n4, InSocketName=calf_l, TransformSpace=RTS_World))
   Set calf_left_location(execute=@n3.then_4, calf_left_location:Vector=@n15) -> (Output_Get:Vector)
   > Set calf_left_rotation(calf_left_rotation:Rotator=@n15.Rotation) -> (Output_Get:Rotator)
   n16 = BreakTransform(InTransform=GetSocketTransform(self=@n4, InSocketName=root, TransformSpace=RTS_World))
   Set body_root_location(execute=@n3.then_6, body_root_location:Vector=@n16) -> (Output_Get:Vector)
   > Set body_root_rotation(body_root_rotation:Rotator=@n16.Rotation) -> (Output_Get:Rotator)
   n17 = Event OnBeginSwimming() {is=title K2Node_Event}
   ResetDynamics(execute=[@n17.then, @n18.then, @n21.then, @n22.then], InTeleportType=TeleportPhysics)
   n18 = Event OnEndSwimming() {is=title K2Node_Event}
   n19 = Event OnNotifyPartsStarted!() -> (OutputDelegate, then, Context:Name=None) {is=title K2Node_Event}
   > Branch(Condition="Equal (Name)"!(A:Name=@n19.Context, B:Name=PhysicsOff) -> (ReturnValue:bool) {is=title K2Node_CallFunction})
   > Set PhysicsActive(PhysicsActive:bool=false) -> (Output_Get:bool)
   n20 = Event OnNotifyPartsStopped!() -> (OutputDelegate, then, Context:Name=None) {is=title K2Node_Event}
   > Branch(Condition="Equal (Name)"(A=@n20.Context, B=PhysicsOff))
   > Set PhysicsActive(PhysicsActive=true)
   n21 = Event OnBeginSwimSprinting() {is=title K2Node_Event}
   n22 = Event OnEndSwimSprinting() {is=title K2Node_Event}
   TryGetPawnOwner()
  ]}
]
variables [
 Body_Mesh {type=SceneComponent, default=null}
 body_root_location {type=Vector}
 body_root_rotation {type=Rotator}
 calf_left_location {type=Vector}
 calf_left_rotation {type=Rotator}
 calf_right_location {type=Vector}
 calf_right_rotation {type=Rotator}
 FrEnd_onPawn {type=boolean}
 NegPhysicsAlpha {type=number}
 pelvis_location {type=Vector}
 pelvis_rotation {type=Rotator}
 PhysicsActive {type=boolean}
 RigidsActive {type=boolean}
 RigidsFE_Alpha {type=number}
 root_location {type=Vector}
 root_space_rotation {type=Rotator}
 rotateRun {type=boolean}
 thigh_left_location {type=Vector}
 thigh_left_rotation {type=Rotator}
 thigh_right_location {type=Vector}
 thigh_right_rotation {type=Rotator}
]
